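# Niagara system for the Pixelatedarrows drift trail vehicle cosmetic, with wheel-wrap and ribbon emitters.
# The wheel-wrap emitter bursts wheel-scaled meshes, partitioned across skeletal mesh bones or sockets.
# Ribbons spawn at quality-scaled rates with a user offset and rearward velocity, inherit vehicle movement on a fading curve, and pass through per-tire disable masking.
# Painted-variant overrides color both emitters, with ribbons using per-variant color-and-alpha curves plus separate R1 overrides.
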NiagaraSystem /VehicleCosmetics/DriftTrail/DriftTrail_Pixelatedarrows/FX/NS_DriftTrail_Pixelatedarrows
dependencies {
 usedDataInterfaces [&NiagaraDataInterfaceColorCurve, &NiagaraDataInterfaceCurve, &NiagaraDataInterfaceSkeletalMesh]
 usedDynamicInputs []
 usedModules [
  &self:SystemSpawnScript.SystemScriptSource.SystemScriptGraph.NiagaraNodeAssignment_0.SetVariables_#1
  &SystemState
 ]
}
emitters {
 NE_VC_DriftTrail_WheelWrap_Pixelated {
  emitterUpdateScript [
   &EmitterState(
    Inactive Response=::NewEnumerator1
    Life Cycle Mode=::NewEnumerator1
    Loop Behavior=::NewEnumerator1
    Loop Duration Mode=::NewEnumerator1
    UseLoopDelay=false
   )
   &SpawnBurst_Instantaneous(Spawn Count=8, Use Spawn Probability=false, UseLoopCountLimit=false)
   &VC_WheelWrapScale()
  ]
  particleSpawnScript [
   &VC_PaintedVariantsOverride(
    BurntSienna=<rgba 0.3854 0.1116 0.0381 1>
    Cobalt=<rgba 0 0.0555 3 1>
    Crimson=<rgba 3 0.125 0.2029 1>
    DefaultColor=<rgba 0.4019 0 2.9688 1>
    ForestGreen=<rgba 0 3 0.1516 1>
    Grey=<rgba 0.3073 0.3073 0.3073 1>
    Lime=<rgba 1.5 3 0 1>
    Orange=<rgba 3 1.0464 0 1>
    Pink=<rgba 3 0.3594 2.0398 1>
    Purple=<rgba 0.343 0 3 1>
    Saffron=<rgba 3 2.3365 0.1562 1>
    SkyBlue=<rgba 0.1875 2.0589 3 1>
   )
   &InitializeParticle(
    Color=$Particles.Color
    Lifetime=$Emitter.CurrentLoopDuration
    Mesh Scale=$Emitter.WheelFXScale
    Mesh Scale Mode=::NewEnumerator4
    Position Offset=<30 0 0>
    Sprite Rotation Mode=::NewEnumerator2
    Sprite Size Mode=::NewEnumerator3
    Uniform Sprite Size=&VectorLength(Vector=<10 10 10>)
    UsePositionOffset=false
    Write Material Random=true
   )
   &PartitionParticles()
   &SkeletalMeshLocation(
    Bone / Socket Index=$Particles.PartitionParticles.Partition
    Bone / Socket Sampling Mode=::NewEnumerator5
    Interpolate Mesh Sampling=false
    MaskSpawnGroup=false
    Mesh Sampling Type=::NewEnumerator6
    Orientation Sampling=::NewEnumerator1
    Skeletal Mesh=$System.SkeletalMesh
   )
  ]
  particleUpdateScript [
   &ParticleState()
   &SolveForcesAndVelocity(Clamp Velocity=false, Limit Acceleration=false)
   &ScaleColor(Scale RGB=&VectorFromFloat(Value=5), ScaleA=true, ScaleRGB=true)
  ]
  rendererClasses [&NiagaraMeshRendererProperties]
  renderers [{rendererClass=&NiagaraMeshRendererProperties, rendererIndex=0}]
 }
 Ribbons {
  emitterUpdateScript [
   &EmitterState()
   &VC_RateQualityHandler(Epic Spawn Rate=50, High Spawn Rate=40, Low Spawn Rate=30, Use Custom Rate?=true)
   &VC_SpawnRibbon_v1(Number of Ribbons=2, SpawnRate=$Emitter.Spawn Rate)
  ]
  particleSpawnScript [
   &VC_PaintedVariantsOverride()
   &VC_InitializeRibbon_v1(
    Color=$Particles.Color
    Custom Offset=true
    Enable Skeletal Position=true
    Local Vector Offset=&AddVectorToPosition(Position=<-25 0 -20>, Vector=$User.offset)
    Offset Amount=1
    Ribbon Width=30
   )
   &AddVelocity(Velocity=<-2000 0 0>, Velocity Speed Scale=2) {module=AddVelocity001}
   /SetVariables(Particles.1-RibbonWidth=50) {isSetParameters=true}
  ]
  particleUpdateScript [
   &FN_DisableTireFX(Mask Index Attribute:NiagaraInt32=$Particles.SpawnGroup, Mask Index FL=2, Mask Index FR=3, Mask Index RL=0, Mask Index RR=1)
   &VC_PaintedVariantsOverride(
    Black=&ColorFromCurve(ColorCurve=DataInterface() {propertyValues {
       AlphaCurve {keys [<at 0 1 RCIM_Cubic>,<at 0.9073 0 RCIM_Cubic>]}
       BlueCurve {keys [<at 0 0.01 RCIM_Cubic>]}
       GreenCurve {keys [<at 0 0.01 RCIM_Cubic>]}
       RedCurve {keys [<at 0 0.01 RCIM_Cubic>]}
      }})
    BurntSienna=&ColorFromCurve(ColorCurve=DataInterface() {propertyValues {
       AlphaCurve {keys [<at 0 1 RCIM_Cubic>,<at 1 0 RCIM_Cubic>]}
       BlueCurve {keys [<at 0 0.0391>,<at 0.5714 0 RCIM_Cubic>]}
       GreenCurve {keys [<at 0 0.1268>,<at 0.5714 0.0164 RCIM_Cubic>]}
       RedCurve {keys [<at 0 0.313>,<at 0.5714 0.1458 RCIM_Cubic>]}
      }})
    Cobalt=&ColorFromCurve(ColorCurve=DataInterface() {propertyValues {
       AlphaCurve {keys [<at 0 1 RCIM_Cubic>,<at 1 0 RCIM_Cubic>]}
       BlueCurve {keys [<at 0 5>,<at 0.5562 0.276 RCIM_Cubic>]}
       GreenCurve {keys [<at 0 0.1992>,<at 0.5562 0.012 RCIM_Cubic>]}
       RedCurve {keys [<at 0 0>,<at 0.5562 0 RCIM_Cubic>]}
      }})
    Crimson=&ColorFromCurve(ColorCurve=DataInterface() {propertyValues {
       AlphaCurve {keys [<at 0 1 RCIM_Cubic>,<at 1 0 RCIM_Cubic>]}
       BlueCurve {keys [<at 0 0.1562>,<at 0.5638 0 RCIM_Cubic>]}
       GreenCurve {keys [<at 0 0.1562>,<at 0.5638 0.0033 RCIM_Cubic>]}
       RedCurve {keys [<at 0 5>,<at 0.5638 0.1979 RCIM_Cubic>]}
      }})
    DefaultColor=&ColorFromCurve(ColorCurve=DataInterface() {propertyValues {
       AlphaCurve {keys [<at 0 0.3>, {interpMode=RCIM_Cubic, leaveTangent=-0.3, time=0.4269, value=0.1}, <at 1 0 RCIM_Cubic>]}
       BlueCurve {keys [<at 0.0483 5>,<at 0.1469 7.7604>,<at 0.3027 0.2604>,<at 0.4328 0>,<at 0.5551 0 RCIM_Cubic>,<at 0.8691 0>]}
       GreenCurve {keys [
         <at 0.0483 0>
         <at 0.1469 1.0417>
         <at 0.3027 0.2604>
         <at 0.4328 3.9511>
         {arriveTangent=-8.7345, interpMode=RCIM_Cubic, leaveTangent=-8.7345, time=0.5551, value=5}
         <at 0.8691 0.1406>
        ]}
       RedCurve {keys [
         <at 0.0483 0.1158>
         <at 0.1469 10>
         <at 0.3027 10>
         <at 0.4328 5>
         {arriveTangent=-11.4483, interpMode=RCIM_Cubic, leaveTangent=-11.4483, time=0.5551, value=0.2026}
         <at 0.8691 0.0057>
        ]}
      }})
    ForestGreen=&ColorFromCurve(ColorCurve=DataInterface() {propertyValues {
       AlphaCurve {keys [<at 0 1 RCIM_Cubic>,<at 1 0 RCIM_Cubic>]}
       BlueCurve {keys [<at 0 0.125>,<at 0.5734 0.013 RCIM_Cubic>]}
       GreenCurve {keys [<at 0 3>,<at 0.5734 0.1927 RCIM_Cubic>]}
       RedCurve {keys [<at 0 0.1859>,<at 0.5734 0.0096 RCIM_Cubic>]}
      }})
    Grey=&ColorFromCurve(ColorCurve=DataInterface() {propertyValues {
       AlphaCurve {keys [<at 0 1 RCIM_Cubic>,<at 1 0 RCIM_Cubic>]}
       BlueCurve {keys [<at 0 0.3021>,<at 0.5734 0.0677 RCIM_Cubic>]}
       GreenCurve {keys [<at 0 0.3021>,<at 0.5734 0.0677 RCIM_Cubic>]}
       RedCurve {keys [<at 0 0.3021>,<at 0.5734 0.0677 RCIM_Cubic>]}
      }})
    Lime=&ColorFromCurve(ColorCurve=DataInterface() {propertyValues {
       AlphaCurve {keys [<at 0 1 RCIM_Cubic>,<at 1 0 RCIM_Cubic>]}
       BlueCurve {keys [<at 0 0.0692>,<at 0.5734 0 RCIM_Cubic>]}
       GreenCurve {keys [<at 0 0.7812>,<at 0.5734 0.1823 RCIM_Cubic>]}
       RedCurve {keys [<at 0 0.5191>,<at 0.5734 0.0977 RCIM_Cubic>]}
      }})
    Orange=&ColorFromCurve(ColorCurve=DataInterface() {propertyValues {
       AlphaCurve {keys [<at 0 1 RCIM_Cubic>,<at 1 0 RCIM_Cubic>]}
       BlueCurve {keys [<at 0 0>,<at 0.5734 0.0056 RCIM_Cubic>]}
       GreenCurve {keys [<at 0 0.5256>,<at 0.5734 0.0796 RCIM_Cubic>]}
       RedCurve {keys [<at 0 3>,<at 0.5734 0.6458 RCIM_Cubic>]}
      }})
    Pink=&ColorFromCurve(ColorCurve=DataInterface() {propertyValues {
       AlphaCurve {keys [<at 0 1 RCIM_Cubic>,<at 1 0 RCIM_Cubic>]}
       BlueCurve {keys [<at 0 0.7663>,<at 0.5734 0.2137 RCIM_Cubic>]}
       GreenCurve {keys [<at 0 0.3337>,<at 0.5734 0.033 RCIM_Cubic>]}
       RedCurve {keys [<at 0 0.7812>,<at 0.5734 0.25 RCIM_Cubic>]}
      }})
    Purple=&ColorFromCurve(ColorCurve=DataInterface() {propertyValues {
       AlphaCurve {keys [<at 0 1 RCIM_Cubic>,<at 1 0 RCIM_Cubic>]}
       BlueCurve {keys [<at 0 0.849>,<at 0.5734 0.3177 RCIM_Cubic>]}
       GreenCurve {keys [<at 0 0.283>,<at 0.5734 0.0387 RCIM_Cubic>]}
       RedCurve {keys [<at 0 0.3941>,<at 0.5734 0.1095 RCIM_Cubic>]}
      }})
    R1 Cobalt:LinearColor=&ColorFromCurve(ColorCurve=DataInterface() {propertyValues {
       AlphaCurve {keys [<at 0 1 RCIM_Cubic>,<at 0.4741 0 RCIM_Cubic>]}
       BlueCurve {keys [<at 0 1 RCIM_Cubic>]}
       GreenCurve {keys [<at 0 0.0185 RCIM_Cubic>]}
       RedCurve {keys [<at 0 0 RCIM_Cubic>]}
      }})
    R1 Crimson:LinearColor=&ColorFromCurve(ColorCurve=DataInterface() {propertyValues {
       AlphaCurve {keys [<at 0 1 RCIM_Cubic>,<at 0.4741 0 RCIM_Cubic>]}
       BlueCurve {keys [<at 0 0.0123 RCIM_Cubic>]}
       GreenCurve {keys [<at 0 0 RCIM_Cubic>]}
       RedCurve {keys [<at 0 0.4621 RCIM_Cubic>]}
      }})
    R1 DefaultColor=&ColorFromCurve(ColorCurve=DataInterface() {propertyValues {
       AlphaCurve {keys [
         <at 0 0.5>
         <at 0.1154 0.3342>
         <at 0.3106 1>
         {arriveTangent=-1.6444, interpMode=RCIM_Cubic, leaveTangent=-1.6444, time=0.4802, value=0.1}
         <at 0.9187 0 RCIM_Cubic>
        ]}
       BlueCurve {keys [
         <at 0.0621 10>
         <at 0.1469 1.5>
         <at 0.3027 0.0415>
         <at 0.4664 0.25>
         {arriveTangent=-0.6208, interpMode=RCIM_Cubic, leaveTangent=-0.6208, time=0.6675, value=0.1185}
         <at 0.8691 0>
        ]}
       GreenCurve {keys [
         <at 0.0621 0>
         <at 0.1469 0>
         <at 0.3027 0.0415>
         <at 0.4664 0.8427>
         {arriveTangent=-1.7432, interpMode=RCIM_Cubic, leaveTangent=-1.7432, time=0.6675, value=0.8125}
         <at 0.8691 0.1406>
        ]}
       RedCurve {keys [
         <at 0.0621 0.8587>
         <at 0.1469 2>
         <at 0.3027 0.8854>
         <at 0.4664 1>
         {arriveTangent=-2.4689, interpMode=RCIM_Cubic, leaveTangent=-2.4689, time=0.6675, value=0.1466}
         <at 0.8691 0.0057>
        ]}
      }})
    R1 ForestGreen:LinearColor=&ColorFromCurve(ColorCurve=DataInterface() {propertyValues {
       AlphaCurve {keys [<at 0 1 RCIM_Cubic>,<at 0.4741 0 RCIM_Cubic>]}
       BlueCurve {keys [<at 0 0.0252 RCIM_Cubic>]}
       GreenCurve {keys [<at 0 0.5029 RCIM_Cubic>]}
       RedCurve {keys [<at 0 0 RCIM_Cubic>]}
      }})
    R1 Grey:LinearColor=&ColorFromCurve(ColorCurve=DataInterface() {propertyValues {
       AlphaCurve {keys [<at 0 1 RCIM_Cubic>,<at 0.4741 0 RCIM_Cubic>]}
       BlueCurve {keys [<at 0 0.1458 RCIM_Cubic>]}
       GreenCurve {keys [<at 0 0.1458 RCIM_Cubic>]}
       RedCurve {keys [<at 0 0.1458 RCIM_Cubic>]}
      }})
    R1 Lime:LinearColor=&ColorFromCurve(ColorCurve=DataInterface() {propertyValues {
       AlphaCurve {keys [<at 0 1 RCIM_Cubic>,<at 0.4741 0 RCIM_Cubic>]}
       BlueCurve {keys [<at 0 0 RCIM_Cubic>]}
       GreenCurve {keys [<at 0 0.5906 RCIM_Cubic>]}
       RedCurve {keys [<at 0 0.2961 RCIM_Cubic>]}
      }})
    R1 Orange:LinearColor=&ColorFromCurve(ColorCurve=DataInterface() {propertyValues {
       AlphaCurve {keys [<at 0 1 RCIM_Cubic>,<at 0.4741 0 RCIM_Cubic>]}
       BlueCurve {keys [<at 0 0 RCIM_Cubic>]}
       GreenCurve {keys [<at 0 0.2346 RCIM_Cubic>]}
       RedCurve {keys [<at 0 0.6724 RCIM_Cubic>]}
      }})
    R1 Pink:LinearColor=&ColorFromCurve(ColorCurve=DataInterface() {propertyValues {
       AlphaCurve {keys [<at 0 1 RCIM_Cubic>,<at 0.4741 0 RCIM_Cubic>]}
       BlueCurve {keys [<at 0 0.3419 RCIM_Cubic>]}
       GreenCurve {keys [<at 0 0.0612 RCIM_Cubic>]}
       RedCurve {keys [<at 0 0.5029 RCIM_Cubic>]}
      }})
    R1 Purple:LinearColor=&ColorFromCurve(ColorCurve=DataInterface() {propertyValues {
       AlphaCurve {keys [<at 0 1 RCIM_Cubic>,<at 0.4741 0 RCIM_Cubic>]}
       BlueCurve {keys [<at 0 0.6795 RCIM_Cubic>]}
       GreenCurve {keys [<at 0 0 RCIM_Cubic>]}
       RedCurve {keys [<at 0 0.2789 RCIM_Cubic>]}
      }})
    R1 Saffron:LinearColor=&ColorFromCurve(ColorCurve=DataInterface() {propertyValues {
       AlphaCurve {keys [<at 0 1 RCIM_Cubic>,<at 0.4741 0 RCIM_Cubic>]}
       BlueCurve {keys [<at 0 0 RCIM_Cubic>]}
       GreenCurve {keys [<at 0 0.6867 RCIM_Cubic>]}
       RedCurve {keys [<at 0 0.8963 RCIM_Cubic>]}
      }})
    R1 SkyBlue:LinearColor=&ColorFromCurve(ColorCurve=DataInterface() {propertyValues {
       AlphaCurve {keys [<at 0 1 RCIM_Cubic>,<at 0.4741 0 RCIM_Cubic>]}
       BlueCurve {keys [<at 0 0.5333 RCIM_Cubic>]}
       GreenCurve {keys [<at 0 0.3564 RCIM_Cubic>]}
       RedCurve {keys [<at 0 0 RCIM_Cubic>]}
      }})
    R1 TitaniumWhite:LinearColor=&ColorFromCurve(ColorCurve=DataInterface() {propertyValues {
       AlphaCurve {keys [<at 0 1 RCIM_Cubic>,<at 0.4608 0 RCIM_Cubic>]}
       BlueCurve {keys [<at 0 1 RCIM_Cubic>,<at 1 1 RCIM_Cubic>]}
       GreenCurve {keys [<at 0 1 RCIM_Cubic>,<at 1 1 RCIM_Cubic>]}
       RedCurve {keys [<at 0 1 RCIM_Cubic>,<at 1 1 RCIM_Cubic>]}
      }})
    Saffron=&ColorFromCurve(ColorCurve=DataInterface() {propertyValues {
       AlphaCurve {keys [<at 0 1 RCIM_Cubic>,<at 1 0 RCIM_Cubic>]}
       BlueCurve {keys [<at 0 0.1665>,<at 0.5734 0.025 RCIM_Cubic>]}
       GreenCurve {keys [<at 0 0.8673>,<at 0.5734 0.1515 RCIM_Cubic>]}
       RedCurve {keys [<at 0 0.8906>,<at 0.5734 0.1719 RCIM_Cubic>]}
      }})
    SkyBlue=&ColorFromCurve(ColorCurve=DataInterface() {propertyValues {
       AlphaCurve {keys [<at 0 1 RCIM_Cubic>,<at 1 0 RCIM_Cubic>]}
       BlueCurve {keys [<at 0 0.7812>,<at 0.5734 0.2969 RCIM_Cubic>]}
       GreenCurve {keys [<at 0 0.7214>,<at 0.5734 0.1632 RCIM_Cubic>]}
       RedCurve {keys [<at 0 0.3393>,<at 0.5734 0.0784 RCIM_Cubic>]}
      }})
    TitaniumWhite=&ColorFromCurve(ColorCurve=DataInterface() {propertyValues {
       AlphaCurve {keys [<at 0 1 RCIM_Cubic>,<at 1 0 RCIM_Cubic>]}
       BlueCurve {keys [<at 0 0.7812>,<at 0.5734 0.2135 RCIM_Cubic>]}
       GreenCurve {keys [<at 0 0.7812>,<at 0.5734 0.2135 RCIM_Cubic>]}
       RedCurve {keys [<at 0 0.7812>,<at 0.5734 0.2135 RCIM_Cubic>]}
      }})
    Use R1 Painted Overrides=true
   ) {module=VC_PaintedVariantsOverride001}
   &ParticleState()
   &InheritSourceMovement(
    Applied Position Scale=&VectorFromFloat(Value=&FloatFromCurve(FloatCurve=DataInterface() {propertyValues {Curve {keys [<at 0.3 1 RCIM_Cubic>, {tangentMode=RCTM_User, time=1, value=0}]}}}))
    ApplyForce=false
    ApplyPosition=true
    ApplyPositionFalloff=false
   )
   &SolveForcesAndVelocity(Clamp Velocity=false, Limit Acceleration=false)
   &VC_UpdateRibbon_v1(Enable Skeletal Position=true)
  ]
  rendererClasses [&NiagaraRibbonRendererProperties]
  renderers [
   {rendererClass=&NiagaraRibbonRendererProperties}
   {rendererClass=&NiagaraRibbonRendererProperties, rendererIndex=1}
  ]
 }
}
summary {
 systemName=NS_DriftTrail_Pixelatedarrows
 emitters [
  {emitterName=NE_VC_DriftTrail_WheelWrap_Pixelated, rendererClasses [&NiagaraMeshRendererProperties]}
  {emitterName=Ribbons, rendererClasses [&NiagaraRibbonRendererProperties]}
 ]
 userVariables [
  User.Alpha_line {type=NiagaraFloat, defaultValue=1, description=""}
  User.BackTrail_power {type=NiagaraFloat, defaultValue=0.5, description=""}
  User.bIsFrontendPreview {type=NiagaraBool}
  User.bIsPainted {type=NiagaraBool}
  User.color intesity {type=NiagaraFloat, defaultValue=60, description=""}
  User.ColorVariantID {type=NiagaraInt32}
  User.offset {type=Vector3f, description="offset from wheels(set through cosmetic template)", defaultValue {struct=Vector3f, value=<0 0 0>}}
  User.PaintedColor {type=LinearColor, description="", defaultValue {struct=LinearColor, value=<rgba 1 1 1 1>}}
  User.PaintedColor2 {type=LinearColor, description="", defaultValue {struct=LinearColor, value=<rgba 1 1 1 1>}}
  User.PaintedScalar {type=NiagaraFloat, defaultValue=1, description=""}
  User.SingleBoostPos {type=NiagaraPosition, description="", defaultValue {}}
  User.WheelSize {type=NiagaraFloat, defaultValue=0.1, description=""}
 ]
}
